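# Physics asset for the FabulousWind medium female player body, covering dynamic pants bones on the right lower leg.
# Sphere bodies sit on the element and global pants bones, with capsules on the calf and its two twist bones.
# Constraints tie the element bone to calf twist 01 and the global bone to calf twist 02.
# Both constraints lock swing1 and twist, leaving swing2 as the only unlocked angular axis.
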PhysicsAsset /BRCosmetics/Characters/Player/Female/Medium/Bodies/F_MED_FabulousWind/Meshes/F_MED_FabulousWind_Physics
bodies {
 Dyn_pants_element_r [{shapeType=Sphere, radius=3.0385, shapeName=Dyn_pants_element_r_sphere, center=<4.849 0 0>}]
 Dyn_pants_global_r [{shapeType=Sphere, radius=3.0385, shapeName=Dyn_pants_element_r_sphere, center=<4.849 0 0>}]
 calf_r [{
   shapeType=Capsule
   length=16.973
   radius=5.9952
   shapeName=thigh_r_capsule
   center=<9.3782 0.9688 -0.1726>
   rotation=<rot 84.6825 179.9985 177.5663>
  }]
 calf_twist_01_r [{
   shapeType=Capsule
   length=7.4336
   radius=10.315
   shapeName=calf_twist_01_r_capsule
   center=<2.2959 0 0>
   rotation=<rot -90 -165.149 165.1487>
  }]
 calf_twist_02_r [{shapeType=Capsule, length=5.1585, radius=8.1044, shapeName=calf_twist_02_r_capsule, center=<5.114 0 0>, rotation=<rot 90 179.9999 -179.9999>}]
}
constraints [
 {
  bone1Name=Dyn_pants_element_r
  bone2Name=calf_twist_01_r
  swing1LimitDegrees=10
  swing1Motion=Locked
  swing2LimitDegrees=25
  twistLimitDegrees=45
  twistMotion=Locked
 }
 {
  bone1Name=Dyn_pants_global_r
  bone2Name=calf_twist_02_r
  swing1LimitDegrees=10
  swing1Motion=Locked
  swing2LimitDegrees=25
  twistLimitDegrees=45
  twistMotion=Locked
 }
]
