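# Physics asset for the FabulousWind medium female player cosmetic body, built around a five-bone dynamic belt chain.
# Constraints link Dyn_Belt01 through Dyn_Belt05 in series from the pelvis, with twist locked on the top two joints and one swing axis locked at the pelvis.
# Collision shapes cover pelvis, lower spine, both thighs, right calf and right forearm, plus a wide, flat root box at floor level.
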PhysicsAsset /BRCosmetics/Characters/Player/Female/Medium/Bodies/F_MED_FabulousWind/Meshes/F_MED_FabulousWind_Physics1
bodies {
 Dyn_Belt01 [{shapeType=Sphere, radius=1.256, shapeName=Dyn_Belt01_sphere}]
 Dyn_Belt02 [
  {shapeType=Capsule, length=10.2508, radius=2.322, shapeName=Dyn_Belt02_capsule, center=<9.0785 0 -1.9064>, rotation=<rot 90 -90 -140.0004>}
  {
   shapeType=Capsule
   length=9.6219
   radius=2.6134
   shapeName=Dyn_Belt02_capsule
   center=<9.5531 0.6452 -2.219>
   rotation=<rot 90 172.875 -137.1254>
  }
 ]
 Dyn_Belt03 [{shapeType=Capsule, length=6.4168, radius=1.8534, shapeName=Dyn_Belt02_capsule, center=<3.78 0 0>, rotation=<rot 90 179.9997 179.9985>}]
 Dyn_Belt04 [{shapeType=Capsule, length=6.4098, radius=1.8997, shapeName=Dyn_Belt02_capsule, center=<4.8542 0 0>, rotation=<rot 90 0 0>}]
 Dyn_Belt05 [{shapeType=Capsule, length=9.6564, radius=1.7206, shapeName=Dyn_Belt02_capsule, center=<5.3518 0 0>, rotation=<rot 90 0 0>}]
 calf_twist_01_r [{
   shapeType=Capsule
   length=32.3737
   radius=11.4043
   shapeName=calf_twist_01_r_capsule
   center=<-10.6973 0.7643 -0.7232>
   rotation=<rot -84.7572 -0.0006 0.0001>
  }]
 lowerarm_r [{shapeType=Capsule, length=25.8196, radius=4.1804, shapeName=lowerarm_r_capsule, center=<-14.4774 0 0>, rotation=<rot -90 180 180>}]
 pelvis [{shapeType=Capsule, length=8.4731, radius=10.4305, shapeName=pelvis_capsule, center=<-3.028 -2.3618 0>}]
 root [{shapeType=Box, extentX=156.9542, extentY=168.5266, extentZ=10.88, shapeName=root_box, center=<0 0 -4.564>}]
 spine_01 [{shapeType=Capsule, length=7.8311, radius=6.8999, shapeName=spine_01_capsule, center=<1.261 0.6524 0>, rotation=<rot 0 179.9993 -180>}]
 thigh_l [{shapeType=Capsule, length=32.5534, radius=7.183, shapeName=thigh_r_capsule, center=<-22.3264 -0.0518 1.1263>, rotation=<rot 90 180 0.0004>}]
 thigh_r [
  {shapeType=Capsule, length=32.5534, radius=7.183, shapeName=thigh_r_capsule, center=<22.3264 0.0518 -1.1263>, rotation=<rot 90 180 -179.9996>}
  {shapeType=Capsule, length=10.5053, radius=6.7308, shapeName=thigh_r_capsule, center=<9.0947 1.919 0.8867>, rotation=<rot -90 -0.0001 0>}
 ]
}
constraints [
 {bone1Name=Dyn_Belt05, bone2Name=Dyn_Belt04, swing1LimitDegrees=25, swing2LimitDegrees=35, twistLimitDegrees=10}
 {bone1Name=Dyn_Belt04, bone2Name=Dyn_Belt03, swing1LimitDegrees=25, swing2LimitDegrees=35, twistLimitDegrees=10}
 {bone1Name=Dyn_Belt03, bone2Name=Dyn_Belt02, swing1LimitDegrees=25, swing2LimitDegrees=35, twistLimitDegrees=10}
 {bone1Name=Dyn_Belt02, bone2Name=Dyn_Belt01, swing1LimitDegrees=30, swing2LimitDegrees=50, twistLimitDegrees=45, twistMotion=Locked}
 {
  bone1Name=Dyn_Belt01
  bone2Name=pelvis
  swing1LimitDegrees=20
  swing1Motion=Locked
  swing2LimitDegrees=30
  twistLimitDegrees=5
  twistMotion=Locked
 }
]
