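# Physics asset for the FabulousWind female medium body cosmetic, apparently for secondary motion of its feathers and jewelry.
# Ten mirrored feather capsules (Dyn_) at breast, clavicle, shoulder, shoulder-back and back each constrain to a small sphere on a matching Dr_ bone.
# Feather joints lock twist and the second swing axis, leaving a single-axis hinge-like swing.
# The Jewelry capsule joins a large spine_05 capsule with only twist locked.
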PhysicsAsset /BRCosmetics/Characters/Player/Female/Medium/Bodies/F_MED_FabulousWind/Meshes/F_MED_FabulousWind_Physics2
bodies {
 Dr_Feather_back_l [{shapeType=Sphere, radius=1.256, shapeName=Dyn_Belt01_sphere}]
 Dr_Feather_back_r [{shapeType=Sphere, radius=1.256, shapeName=Dyn_Belt01_sphere}]
 Dr_Feather_breast_l [{shapeType=Sphere, radius=1.256, shapeName=Dyn_Belt01_sphere}]
 Dr_Feather_breast_r [{shapeType=Sphere, radius=1.256, shapeName=Dyn_Belt01_sphere}]
 Dr_Feather_clavicle_l [{shapeType=Sphere, radius=1.256, shapeName=Dyn_Belt01_sphere}]
 Dr_Feather_clavicle_r [{shapeType=Sphere, radius=1.256, shapeName=Dyn_Belt01_sphere}]
 Dr_Feather_shoulder_back_l [{shapeType=Sphere, radius=1.256, shapeName=Dyn_Belt01_sphere}]
 Dr_Feather_shoulder_back_r [{shapeType=Sphere, radius=1.256, shapeName=Dyn_Belt01_sphere}]
 Dr_Feather_shoulder_l [{shapeType=Sphere, radius=1.256, shapeName=Dyn_Belt01_sphere}]
 Dr_Feather_shoulder_r [{shapeType=Sphere, radius=1.256, shapeName=Dyn_Belt01_sphere}]
 Dyn_Feather_back_l [{
   shapeType=Capsule
   length=4.0215
   radius=1.4093
   shapeName=Dyn_Feather_breast_r_capsule
   center=<-4.546 0 0>
   rotation=<rot 90 179.9999 0.0002>
  }]
 Dyn_Feather_back_r [{
   shapeType=Capsule
   length=4.0215
   radius=1.4093
   shapeName=Dyn_Feather_breast_r_capsule
   center=<4.546 0 0>
   rotation=<rot 90 179.9999 -179.9998>
  }]
 Dyn_Feather_breast_l [{
   shapeType=Capsule
   length=5.1544
   radius=1.4093
   shapeName=Dyn_Feather_breast_r_capsule
   center=<-3.6574 0 0>
   rotation=<rot 90 179.9999 0.0002>
  }]
 Dyn_Feather_breast_r [{
   shapeType=Capsule
   length=5.1544
   radius=1.4093
   shapeName=Dyn_Feather_breast_r_capsule
   center=<3.6574 0 0>
   rotation=<rot 90 179.9999 -179.9998>
  }]
 Dyn_Feather_clavicle_l [{
   shapeType=Capsule
   length=5.1544
   radius=1.4093
   shapeName=Dyn_Feather_breast_r_capsule
   center=<-4.546 0 0>
   rotation=<rot 90 179.9999 0.0002>
  }]
 Dyn_Feather_clavicle_r [{
   shapeType=Capsule
   length=5.1544
   radius=1.4093
   shapeName=Dyn_Feather_breast_r_capsule
   center=<4.546 0 0>
   rotation=<rot 90 179.9999 -179.9998>
  }]
 Dyn_Feather_shoulder_back_l [{
   shapeType=Capsule
   length=5.1544
   radius=1.4093
   shapeName=Dyn_Feather_breast_r_capsule
   center=<-5.1334 0 0>
   rotation=<rot 90 179.9999 0.0002>
  }]
 Dyn_Feather_shoulder_back_r [{
   shapeType=Capsule
   length=5.1544
   radius=1.4093
   shapeName=Dyn_Feather_breast_r_capsule
   center=<5.1334 0 0>
   rotation=<rot 90 179.9999 -179.9998>
  }]
 Dyn_Feather_shoulder_l [{
   shapeType=Capsule
   length=6.5211
   radius=1.4093
   shapeName=Dyn_Feather_breast_r_capsule
   center=<-5.9661 0 0>
   rotation=<rot 90 179.9999 0.0002>
  }]
 Dyn_Feather_shoulder_r [{
   shapeType=Capsule
   length=6.5211
   radius=1.4093
   shapeName=Dyn_Feather_breast_r_capsule
   center=<5.9661 0 0>
   rotation=<rot 90 179.9999 -179.9998>
  }]
 Jewelry [{shapeType=Capsule, length=2.3332, radius=1.4756, shapeName=Jewelry_capsule, center=<10.3822 1.1085 0>}]
 spine_05 [{shapeType=Capsule, length=10.154, radius=8.773, shapeName=spine_05_capsule, center=<-3.036 1.3661 0>}]
}
constraints [
 {
  bone1Name=Dyn_Feather_breast_r
  bone2Name=Dr_Feather_breast_r
  swing1LimitDegrees=5
  swing2LimitDegrees=45
  swing2Motion=Locked
  twistLimitDegrees=45
  twistMotion=Locked
 }
 {
  bone1Name=Dyn_Feather_clavicle_r
  bone2Name=Dr_Feather_clavicle_r
  swing1LimitDegrees=5
  swing2LimitDegrees=45
  swing2Motion=Locked
  twistLimitDegrees=45
  twistMotion=Locked
 }
 {
  bone1Name=Dyn_Feather_shoulder_r
  bone2Name=Dr_Feather_shoulder_r
  swing1LimitDegrees=5
  swing2LimitDegrees=45
  swing2Motion=Locked
  twistLimitDegrees=45
  twistMotion=Locked
 }
 {
  bone1Name=Dyn_Feather_shoulder_back_r
  bone2Name=Dr_Feather_shoulder_back_r
  swing1LimitDegrees=5
  swing2LimitDegrees=45
  swing2Motion=Locked
  twistLimitDegrees=45
  twistMotion=Locked
 }
 {
  bone1Name=Dyn_Feather_back_r
  bone2Name=Dr_Feather_back_r
  swing1LimitDegrees=5
  swing2LimitDegrees=45
  swing2Motion=Locked
  twistLimitDegrees=45
  twistMotion=Locked
 }
 {
  bone1Name=Dyn_Feather_back_l
  bone2Name=Dr_Feather_back_l
  swing1LimitDegrees=5
  swing2LimitDegrees=45
  swing2Motion=Locked
  twistLimitDegrees=45
  twistMotion=Locked
 }
 {
  bone1Name=Dyn_Feather_shoulder_back_l
  bone2Name=Dr_Feather_shoulder_back_l
  swing1LimitDegrees=5
  swing2LimitDegrees=45
  swing2Motion=Locked
  twistLimitDegrees=45
  twistMotion=Locked
 }
 {
  bone1Name=Dyn_Feather_shoulder_l
  bone2Name=Dr_Feather_shoulder_l
  swing1LimitDegrees=5
  swing2LimitDegrees=45
  swing2Motion=Locked
  twistLimitDegrees=45
  twistMotion=Locked
 }
 {
  bone1Name=Dyn_Feather_clavicle_l
  bone2Name=Dr_Feather_clavicle_l
  swing1LimitDegrees=5
  swing2LimitDegrees=45
  swing2Motion=Locked
  twistLimitDegrees=45
  twistMotion=Locked
 }
 {
  bone1Name=Dyn_Feather_breast_l
  bone2Name=Dr_Feather_breast_l
  swing1LimitDegrees=5
  swing2LimitDegrees=45
  swing2Motion=Locked
  twistLimitDegrees=45
  twistMotion=Locked
 }
 {bone1Name=Jewelry, bone2Name=spine_05, swing1LimitDegrees=12, swing2LimitDegrees=40, twistLimitDegrees=45, twistMotion=Locked}
]
