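# Physics asset for the right-arm feathers on the FabulousWind medium female player body, covering top, bottom, inner and outside feathers.
# Each Dyn_ feather bone has a capsule body constrained to a small sphere body on its matching featherArm bone, with identical narrow swing and twist limits on all four pairs.
# A box body on hand_r has no constraint, apparently serving as a collider for the feathers.
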PhysicsAsset /BRCosmetics/Characters/Player/Female/Medium/Bodies/F_MED_FabulousWind/Meshes/F_MED_FabulousWind_PhysicsRArmFeather
bodies {
 Dyn_featherArmBot_r [{
   shapeType=Capsule
   length=5.0807
   radius=1.6237
   shapeName=Dyn_featherArmTop_r_capsule
   center=<4.778 0 -0.6909>
   rotation=<rot 90 179.9209 179.9209>
  }]
 Dyn_featherArmInner_r [{
   shapeType=Capsule
   length=4.2753
   radius=1.6237
   shapeName=Dyn_featherArmTop_r_capsule
   center=<4.0299 0 -0.6909>
   rotation=<rot 90 179.367 179.367>
  }]
 Dyn_featherArmOutside_r [{
   shapeType=Capsule
   length=4.4318
   radius=1.6237
   shapeName=Dyn_featherArmTop_r_capsule
   center=<4.3129 0 -0.6909>
   rotation=<rot 90 179.9209 179.9209>
  }]
 Dyn_featherArmTop_r [{
   shapeType=Capsule
   length=10.0509
   radius=1.6237
   shapeName=Dyn_featherArmTop_r_capsule
   center=<8.0807 0 -0.6909>
   rotation=<rot 90 169.975 169.975>
  }]
 featherArmBot_r [{shapeType=Sphere, radius=1.1989, shapeName=featherArmTop_r_sphere}]
 featherArmInner_r [{shapeType=Sphere, radius=1.1989, shapeName=featherArmTop_r_sphere}]
 featherArmOutside_r [{shapeType=Sphere, radius=1.1989, shapeName=featherArmTop_r_sphere}]
 featherArmTop_r [{shapeType=Sphere, radius=1.1989, shapeName=featherArmTop_r_sphere}]
 hand_r [{shapeType=Box, extentX=8.1694, extentY=2.4902, extentZ=8.4699, shapeName=hand_r_box, center=<-5.6484 1.0649 0.3391>}]
}
constraints [
 {bone1Name=Dyn_featherArmTop_r, bone2Name=featherArmTop_r, swing1LimitDegrees=10, swing2LimitDegrees=25, twistLimitDegrees=3}
 {bone1Name=Dyn_featherArmOutside_r, bone2Name=featherArmOutside_r, swing1LimitDegrees=10, swing2LimitDegrees=25, twistLimitDegrees=3}
 {bone1Name=Dyn_featherArmBot_r, bone2Name=featherArmBot_r, swing1LimitDegrees=10, swing2LimitDegrees=25, twistLimitDegrees=3}
 {bone1Name=Dyn_featherArmInner_r, bone2Name=featherArmInner_r, swing1LimitDegrees=10, swing2LimitDegrees=25, twistLimitDegrees=3}
]
