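# Post-process style cosmetic AnimBP for the FabulousWind female body, adding secondary motion to the copied parent pose.
# The pose passes through a Control Rig, a backpack-bone offset and five chained RigidBody sims; the pants AnimDynamics chain is not wired to the output.
# Each update feeds calf-driven WPO cloth parameters to body and overlay materials, raising wind and turbulence during skydive or parachute.
# It also refreshes the pelvis sim's gravity override and the front-end/emote flag.
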AnimBlueprint /BRCosmetics/Characters/Player/Female/Medium/Bodies/F_MED_FabulousWind/Meshes/F_MED_FabulousWind_AnimBP
target_skeleton=&/F_MED_FabulousWind_Skeleton
graphs [
 AnimGraph {nodes [
   n1 = Anim Dynamics - Bone: Dyn_pants_element_r(Alpha=1, ExternalForce="0,0,0") {
    angularBiasOverride=0.7
    bGravityOverrideInSimSpace=true
    bLinearSpring=true
    bOverrideAngularBias=true
    bOverrideAngularDamping=true
    bOverrideLinearDamping=true
    bUseGravityOverride=true
    bUsePlanarLimit=false
    linearDampingOverride=0.95
    linearSpringConstant=150
    numSolverIterationsPostUpdate=8
    numSolverIterationsPreUpdate=16
    simulationSpace=BoneRelative
    alphaBoolBlend {blendOption=Linear}
    boundBone {boneName=Dyn_pants_element_r}
    componentAppliedLinearAccClamp=<5 5 5>
    componentLinearAccScale=<1 1 1>
    componentLinearVelScale=<1 1 1>
    physicsBodyDefinitions [{
      boundBone {boneName=Dyn_pants_element_r}
      boxExtents=<10 10 10>
      constraintSetup {angularLimitsMax=<0 0 0>, angularLimitsMin=<0 0 0>, angularTarget=<0 0 0>, linearAxesMax=<1 1 1>, linearAxesMin=<-1 -1 -1>}
      localJointOffset=<0 -10 0>
     }]
    relativeSpaceBone {boneName=thigh_r}
   }
   n2 = Anim Dynamics - Bone: Dyn_pants_global_r(ComponentPose=@n1, Alpha=1, ExternalForce="0,0,0") {
    angularBiasOverride=0.7
    bGravityOverrideInSimSpace=true
    bLinearSpring=true
    bOverrideAngularBias=true
    bOverrideAngularDamping=true
    bOverrideLinearDamping=true
    bUseGravityOverride=true
    bUsePlanarLimit=false
    linearDampingOverride=0.95
    linearSpringConstant=150
    numSolverIterationsPostUpdate=8
    numSolverIterationsPreUpdate=16
    simulationSpace=BoneRelative
    alphaBoolBlend {blendOption=Linear}
    boundBone {boneName=Dyn_pants_global_r}
    componentAppliedLinearAccClamp=<5 5 5>
    componentLinearAccScale=<1 1 1>
    componentLinearVelScale=<1 1 1>
    physicsBodyDefinitions [{
      boundBone {boneName=Dyn_pants_global_r}
      boxExtents=<10 10 10>
      constraintSetup {angularLimitsMax=<0 0 0>, angularLimitsMin=<0 0 0>, angularTarget=<0 0 0>, linearAxesMax=<1 1 1>, linearAxesMin=<-1 -1 -1>}
      localJointOffset=<0 -10 0>
     }]
    relativeSpaceBone {boneName=thigh_r}
   }
   n3 = Anim Dynamics - Bone: Dyn_pants_global_l(ComponentPose=@n2, Alpha=1, ExternalForce="0,0,0") {
    angularBiasOverride=0.7
    bGravityOverrideInSimSpace=true
    bLinearSpring=true
    bOverrideAngularBias=true
    bOverrideAngularDamping=true
    bOverrideLinearDamping=true
    bUseGravityOverride=true
    bUsePlanarLimit=false
    linearDampingOverride=0.9
    linearSpringConstant=50
    numSolverIterationsPostUpdate=8
    numSolverIterationsPreUpdate=16
    simulationSpace=BoneRelative
    alphaBoolBlend {blendOption=Linear}
    boundBone {boneName=Dyn_pants_global_l}
    componentAppliedLinearAccClamp=<5 5 5>
    componentLinearAccScale=<1 1 1>
    componentLinearVelScale=<1 1 1>
    gravityOverride {z=-100}
    physicsBodyDefinitions [{
      boundBone {boneName=Dyn_pants_global_l}
      boxExtents=<5 5 5>
      constraintSetup {angularLimitsMax=<0 0 0>, angularLimitsMin=<0 0 0>, angularTarget=<0 0 0>, linearAxesMax=<0.1 1 0.1>, linearAxesMin=<-0.1 0 -0.1>}
      localJointOffset=<0 -10 0>
     }]
    relativeSpaceBone {boneName=thigh_l}
   }
   Anim Dynamics - Bone: Dyn_pants_element_l(ComponentPose=@n3, Alpha=1, ExternalForce="0,0,0") {
    angularBiasOverride=0.7
    bGravityOverrideInSimSpace=true
    bLinearSpring=true
    bOverrideAngularBias=true
    bOverrideAngularDamping=true
    bOverrideLinearDamping=true
    bUseGravityOverride=true
    bUsePlanarLimit=false
    linearDampingOverride=0.9
    linearSpringConstant=50
    numSolverIterationsPostUpdate=8
    numSolverIterationsPreUpdate=16
    simulationSpace=BoneRelative
    alphaBoolBlend {blendOption=Linear}
    boundBone {boneName=Dyn_pants_element_l}
    componentAppliedLinearAccClamp=<5 5 5>
    componentLinearAccScale=<1 1 1>
    componentLinearVelScale=<1 1 1>
    gravityOverride {z=-100}
    physicsBodyDefinitions [{
      boundBone {boneName=Dyn_pants_element_l}
      boxExtents=<5 5 5>
      constraintSetup {angularLimitsMax=<0 0 0>, angularLimitsMin=<0 0 0>, angularTarget=<0 0 0>, linearAxesMax=<0.1 1 0.1>, linearAxesMin=<-0.1 0 -0.1>}
      localJointOffset=<0 -10 0>
     }]
    relativeSpaceBone {boneName=thigh_l}
   }
   n4 = Control Rig!(
    Source=Copy Pose From Mesh() {bUseAttachedParent=true}
    Alpha
    RBAN:bool=Get bRigidBodyEnabled()
    Is FrontEnd:bool=Get Is FrontEnd Or Emote() -> (Is FrontEnd Or Emote:bool)
   ) -> (Pose) {
    alphaBoolBlend {blendOption=Linear}
    controlRigAssetReference {blueprintRigClass=&/F_MED_FabulousWind_CtrlRig.F_MED_FabulousWind_CtrlRig_C}
   }
   n5 = "Transform (Modify) Bone - Bone: attach_backpack"(ComponentPose=Local To Component(LocalPose=@n4), Alpha=1, Translation="0,2,0", Rotation="0,0,0", Scale="1,1,1") {
    translationMode=BMM_Additive
    translationSpace=BCS_BoneSpace
    alphaBoolBlend {blendOption=Linear}
    boneToModify {boneName=attach_backpack}
    translation {y=2}
   }
   n6 = RigidBody!(ComponentPose=@n5, Alpha, OverrideWorldGravity:Vector="0,0,0", ExternalForce) -> (Pose) {
    bOverrideWorldGravity=true
    overridePhysicsAsset=&/F_MED_FabulousWind_Physics
    simulationSpace=BaseBoneSpace
    alphaBoolBlend {blendOption=Linear}
    baseBoneRef {boneName=calf_r}
    componentAppliedLinearAccClamp=<2000 2000 2000>
    componentLinearAccScale=<1 1 1>
    componentLinearVelScale=<1 1 1>
    simSpaceSettings {maxAngularAcceleration=50, maxAngularVelocity=10, maxLinearAcceleration=50, maxLinearVelocity=10, worldAlpha=1}
   }
   n7 = RigidBody(ComponentPose=@n6, OverrideWorldGravity="0,0,0") {
    bOverrideWorldGravity=true
    overridePhysicsAsset=&/F_MED_FabulousWind_Physics_FeatherLegLeft
    simulationSpace=BaseBoneSpace
    alphaBoolBlend {blendOption=Linear}
    baseBoneRef {boneName=calf_l}
    componentAppliedLinearAccClamp=<2000 2000 2000>
    componentLinearAccScale=<1 1 1>
    componentLinearVelScale=<1 1 1>
    simSpaceSettings {maxAngularAcceleration=50, maxAngularVelocity=10, maxLinearAcceleration=50, maxLinearVelocity=10, worldAlpha=1}
   }
   n8 = RigidBody(ComponentPose=@n7, OverrideWorldGravity=Get Gravity Override()) {
    bOverrideWorldGravity=true
    overridePhysicsAsset=&/F_MED_FabulousWind_Physics1
    simulationSpace=BaseBoneSpace
    alphaBoolBlend {blendOption=Linear}
    baseBoneRef {boneName=pelvis}
    componentAppliedLinearAccClamp=<600 600 600>
    componentLinearAccScale=<1 1 1>
    componentLinearVelScale=<1 1 1>
    simSpaceSettings {maxAngularAcceleration=25, maxAngularVelocity=5, maxLinearAcceleration=50, maxLinearVelocity=15, velocityScaleZ=0, worldAlpha=0.6}
   }
   n9 = RigidBody!(ComponentPose=@n8, Alpha, ExternalForce) -> (Pose) {
    overridePhysicsAsset=&/F_MED_FabulousWind_Physics2
    simulationSpace=BaseBoneSpace
    alphaBoolBlend {blendOption=Linear}
    baseBoneRef {boneName=spine_04}
    componentAppliedLinearAccClamp=<300 300 300>
    simSpaceSettings {maxAngularAcceleration=75, maxAngularVelocity=50, maxLinearAcceleration=75, maxLinearVelocity=50, velocityScaleZ=0, worldAlpha=0.2}
   }
   n10 = RigidBody(ComponentPose=@n9) {
    overridePhysicsAsset=&/F_MED_FabulousWind_PhysicsRArmFeather
    simulationSpace=BaseBoneSpace
    alphaBoolBlend {blendOption=Linear}
    baseBoneRef {boneName=hand_r}
    componentAppliedLinearAccClamp=<300 300 300>
    simSpaceSettings {maxAngularAcceleration=25, maxAngularVelocity=3, maxLinearAcceleration=25, maxLinearVelocity=3, velocityScaleZ=0, worldAlpha=0.2}
   }
   Output Pose(Result=Component To Local(ComponentPose=@n10)) {layerGroup=None}
  ]}
 EventGraph {nodes [
   n1 = Event BlueprintUpdateAnimation()
   > n2 = Sequence!(execute) -> (then_0, then_1, then_2:Exec)
   n3 = Get Mesh!() -> (Mesh)
   n4 = Get Overlay Mat Instance Dyanmic() -> (Overlay Mat Instance Dyanmic:Material Instance Dynamic Object Reference)
   n5 = Get Material Instance Dynamic() -> (Material Instance Dynamic:Material Instance Dynamic Object Reference)
   > Set WPO Globals Function(
    Material=@n5
    Overlay Material=@n4
    Skeletal Mesh=@n3
    "Local Noise Transform  Location"="Get Local Noise Transform  Location"() -> ("Local Noise Transform  Location":Vector)
    Local Noise Offset=Get Local Noise Offset()
   )
   > Set WPO Regional Function!(
    execute
    Material:Material Instance Dynamic Object Reference=@n5
    Overlay Material:Material Instance Dynamic Object Reference=@n4
    Skeletal Mesh:Skeletal Mesh Component Object Reference=@n3
    Velocity Transform:Name=calf_r
    Velocity Interp Speed:double=Get Velocity Interp Speed()
    Collision Transform:Name=calf_r
    Collision Transform Axis:EAxis Enum
    Invert Collision Transform Axis:bool=false
    Delta Time X:double=@n1.DeltaTimeX
    Previous Velocity Transform Location:"Vector (by ref)"=Get R2 Velocity Transform Location()
    Previous Velocity Transform Velocity:"Vector (by ref)"=Get R2 Velocity Transform Velocity()
    Region:int=1
   ) -> (then)
   > Set WPO Regional Function(
    Material=@n5
    Overlay Material=@n4
    Skeletal Mesh=@n3
    Velocity Transform=calf_l
    Velocity Interp Speed=Get Velocity Interp Speed()
    Collision Transform=calf_l
    Invert Collision Transform Axis=true
    Delta Time X=@n1.DeltaTimeX
    Previous Velocity Transform Location=Get L Previous Velocity Transform Location() -> (L Previous Velocity Transform Location:Vector)
    Previous Velocity Transform Velocity=Get L Previous Velocity Transform Velocity() -> (L Previous Velocity Transform Velocity:Vector)
    Region=2
   )
   > SetScalarParameterValue(self=[@n5, @n4], ParameterName=ClothSim_WindStrength, Value=Get WPO Wind() -> (WPO Wind:double))
   > SetScalarParameterValue(self=[@n5, @n4], ParameterName=ClothSim_TurbulenceNoiseTileDivisor, Value=Get Turbulence Noise() -> (Turbulence Noise:float))
   n6 = Branch(
    execute=@n2.then_1
    Condition=OR Boolean!(A=Get bIsSkydiveFloating(), B=Get bIsSkydiveDiveMode(), C:bool=Get bIsParachuteOpened()) -> (ReturnValue)
   )
   > Set Velocity Interp Speed(Velocity Interp Speed:double=40) -> (Output_Get:double)
   > Set WPO Wind(WPO Wind:double=1) -> (Output_Get:double)
   > Set Turbulence Noise(Turbulence Noise:float=240) -> (Output_Get:float)
   Set Velocity Interp Speed(execute=@n6.else, Velocity Interp Speed=0)
   > Set WPO Wind(WPO Wind=0)
   > Set Turbulence Noise(Turbulence Noise=0)
   n7 = Gravity Override State Function DEPRECATED(
    execute=@n2.then_2
    CharacterPartAnimInstance=Self-Reference()
    GameTimeInSeconds=GetGameTimeInSeconds()
    GravityOverrideParams=Get Gravity Override Params()
   )
   > Set Gravity Override(Gravity Override=@n7)
   > Set Is FrontEnd Or Emote(Is FrontEnd Or Emote:bool=OR Boolean!(A=Get bIsInFrontEnd(), B=Get bIsPlayingEmote()) -> (ReturnValue)) -> (Output_Get:bool)
   Event OnCharacterCustomizationLoaded()
   n8 = GetPartSkeletalMeshComponent()
   > n9 = Set Mesh(Mesh:Skeletal Mesh Component Object Reference=@n8) -> (Output_Get:Skeletal Mesh Component Object Reference)
   > n10 = Get Dynamic Material Instance(SkeletalMeshComponent=@n9)
   > Set Material Instance Dynamic(Material Instance Dynamic:Material Instance Dynamic Object Reference=@n10) -> (Output_Get:Material Instance Dynamic Object Reference)
   > n11 = GetOrCreateDynamicOverlayMaterial(MeshComponent=@n8, MaterialName=None)
   > Set Overlay Mat Instance Dyanmic(Overlay Mat Instance Dyanmic:Material Instance Dynamic Object Reference=@n11) -> (Output_Get:Material Instance Dynamic Object Reference)
  ]}
]
variables [
 Gravity Override {type=Vector}
 Gravity Override Params {
  type=GravityOverrideParamsStruct
  default {
   skydive_FloatingDown {gravityOverride=<0 0 0>}
   skydive_GoingUp {gravityOverride=<0 0 0>}
   skydive_Parachuting {gravityOverride=<0 0 -980>, windAmplitude=<0 0 0>, windFrequency=<0 0 0>}
  }
 }
 Is FrontEnd Or Emote {type=boolean}
 L Previous Velocity Transform Location {type=Vector}
 L Previous Velocity Transform Velocity {type=Vector}
 Local Noise Offset {type=Vector}
 "Local Noise Transform  Location" {type=Vector}
 Material Instance Dynamic {type=MaterialInstanceDynamic}
 Mesh {type=SkeletalMeshComponent}
 Overlay Mat Instance Dyanmic {type=MaterialInstanceDynamic}
 R2 Velocity Transform Location {type=Vector}
 R2 Velocity Transform Velocity {type=Vector}
 Turbulence Noise {type=number}
 Velocity Interp Speed {type=number}
 WPO Wind {type=number}
]
