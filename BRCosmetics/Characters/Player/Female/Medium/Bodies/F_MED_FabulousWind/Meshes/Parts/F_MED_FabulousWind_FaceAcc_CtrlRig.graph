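# Control Rig for the F_MED_FabulousWind face accessory that poses hair tail, clavicle strands, bangs and back-hat bones to follow body motion.
# Head rotation relative to spine_05 and clavicle rotations are remapped into weights for constraints and corrective offsets, run in sequence.
# With IsWearingBackbling, tail bones pin to the spine, plus a head pose-reader nudge scaled by pelvis-root distance.
# When RBAN is false, Dyn_HairTail bones are constrained to the spine.
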ControlRigBlueprint /BRCosmetics/Characters/Player/Female/Medium/Bodies/F_MED_FabulousWind/Meshes/Parts/F_MED_FabulousWind_FaceAcc_CtrlRig
graphs [
 RigVMFunctionLibrary {links [], nodes []}
 RigVMFunction_Sequence_2_1_ContainedGraph {nodes [
   n1 = RigVMFunctionEntryNode()
   n2 = &RigVMFunction_Sequence(ExecuteContext=@n1.ExecuteContext)
   n3 = &RigVMFunction_Sequence(ExecuteContext=@n2.B)
   n4 = &RigVMFunction_Sequence(ExecuteContext=@n3.B)
   n5 = &RigVMFunction_Sequence(ExecuteContext=@n4.B)
   n6 = &RigVMFunction_Sequence(ExecuteContext=@n5.B)
   n7 = &RigVMFunction_Sequence(ExecuteContext=@n6.B)
   RigVMFunctionReturnNode(A=@n2.A, B=@n3.A, C=@n4.A, D=@n5.A, E=@n6.A, F=@n7.A, G=@n7.B)
  ]}
 RigVMModel {nodes [
   RigUnit_BeginExecution = &RigUnit_BeginExecution()
   RigVMVariableNode(
    ExecuteContext=@RigUnit_BeginExecution.ExecutePin
    Value=&RigVMFunction_MathQuaternionToEuler(
     RotationOrder=ZYX
     Value=&RigUnit_GetRelativeTransformForItem(Child=(Type=Bone,Name="head"), Parent=(Type=Bone,Name="spine_05"), bChildInitial=False, bParentInitial=False).RelativeTransform.Rotation
    ).Result
   ) {variable=HeadRot}
   > RigVMFunction_Sequence_2 = RigVMAggregateNode()
   &RigUnit_AimConstraintLocalSpaceOffset(
    Child=(Type=Bone,Name="Dr_HairTailRoot")
    Filter=(bX=true,bY=True,bZ=True)
    Parents=((Item=(Type=Bone,Name="spine_05"),Weight=1))
    Weight=1
    WorldUp=(Target=(X=0,Y=0,Z=1),Kind=Location,Space=(Type=Bone,Name="head"))
    bMaintainOffset=true
    ExecutePin=@RigVMFunction_Sequence_2.A
   )
   n1 = RigVMVariableNode() {variable=HeadRot}
   n2 = &RigVMFunction_MathDoubleRemap(SourceMaximum=20, SourceMinimum=0, TargetMaximum=1, TargetMinimum=0, bClamp=False, Value=@n1.Value.Z)
   > &RigUnit_ModifyTransforms(
    ItemToModify=((Item=(Type=Bone,Name="Dr_HairTailRoot"),Transform=(Rotation=(X=0,Y=0,Z=-0.0872,W=0.9962),Translation=(X=3,Y=2,Z=0),Scale3D=(X=1,Y=1,Z=1))))
    WeightMinimum=0
    Weight=@n2.Result
   )
   Remap = &RigVMFunction_MathDoubleRemap(SourceMaximum=20, SourceMinimum=-20, TargetMaximum=1, TargetMinimum=0, Value=RigVMVariableNode() {variable=HeadRot}.Value.X)
   Subtract = &RigVMFunction_MathDoubleSub(A=1, B=@Remap.Result)
   &RigUnit_ParentConstraint(
    AdvancedSettings=(InterpolationType=Average,RotationOrderForFilter=XZY)
    Child=(Type=Bone,Name="Dr_HairClaviclr_r")
    Filter=(TranslationFilter=(bX=True,bY=True,bZ=True),RotationFilter=(bX=true,bY=true,bZ=true),ScaleFilter=(bX=True,bY=True,bZ=True))
    Parents=((Item=(Type=Bone,Name="clavicle_r"),Weight=1))
    ExecutePin=@RigVMFunction_Sequence_2.B
    Weight=@Subtract.Result
   )
   > &RigUnit_ParentConstraint(
    AdvancedSettings=(InterpolationType=Average,RotationOrderForFilter=XZY)
    Child=(Type=Bone,Name="Dr_HairClaviclr_l")
    Filter=(TranslationFilter=(bX=True,bY=True,bZ=True),RotationFilter=(bX=true,bY=true,bZ=true),ScaleFilter=(bX=True,bY=True,bZ=True))
    Parents=((Item=(Type=Bone,Name="clavicle_l"),Weight=1))
    Weight=@Remap.Result
   )
   n3 = RigVMVariableNode() {variable=HeadRot}
   n4 = &RigVMFunction_MathDoubleRemap(SourceMaximum=-90, SourceMinimum=0, TargetMaximum=1, TargetMinimum=0, Value=@n3.Value.X)
   > &RigUnit_ParentConstraint(
    AdvancedSettings=(InterpolationType=Average,RotationOrderForFilter=XZY)
    Child=(Type=Bone,Name="Dr_HairClaviclr_l")
    Filter=(TranslationFilter=(bX=true,bY=true,bZ=true),RotationFilter=(bX=true,bY=true,bZ=true),ScaleFilter=(bX=True,bY=True,bZ=True))
    Parents=((Item=(Type=Bone,Name="clavicle_l"),Weight=1))
    Weight=@n4.Result
   )
   n5 = &RigVMFunction_MathDoubleRemap(SourceMaximum=90, SourceMinimum=0, TargetMaximum=1, TargetMinimum=0, Value=@n3.Value.X)
   > &RigUnit_ParentConstraint(
    AdvancedSettings=(InterpolationType=Average,RotationOrderForFilter=XZY)
    Child=(Type=Bone,Name="Dr_HairClaviclr_r")
    Parents=((Item=(Type=Bone,Name="clavicle_r"),Weight=1))
    Weight=@n5.Result
   )
   &RigUnit_ParentConstraint(
    AdvancedSettings=(InterpolationType=Average,RotationOrderForFilter=XZY)
    Child=(Type=Bone,Name="Dr_HairTailMain")
    Parents=((Item=(Type=Bone,Name="spine_04"),Weight=1))
    Weight=0.8
    ExecutePin=@RigVMFunction_Sequence_2.C
   )
   > &RigUnit_ModifyTransforms(
    ItemToModify=((Item=(Type=Bone,Name="Dr_HairTailMain"),Transform=(Rotation=(X=0,Y=0,Z=0,W=1),Translation=(X=3,Y=-3,Z=0),Scale3D=(X=1,Y=1,Z=1))))
    WeightMinimum=0
    Weight=&RigVMFunction_MathDoubleRemap(SourceMaximum=-30, SourceMinimum=0, TargetMaximum=1, TargetMinimum=0, bClamp=False, Value=@n1.Value.Z).Result
   )
   > &RigUnit_ModifyTransforms(
    ItemToModify=((Item=(Type=Bone,Name="Dr_HairTailMain"),Transform=(Rotation=(X=0,Y=0,Z=-0.1305,W=0.9914),Translation=(X=-1,Y=2,Z=0),Scale3D=(X=1,Y=1,Z=1))))
    WeightMinimum=0
    Weight=@n2.Result
   )
   n6 = RigVMVariableNode() {variable=HeadRot}
   n7 = &RigVMFunction_MathDoubleRemap(SourceMaximum=60, SourceMinimum=0, TargetMaximum=1, TargetMinimum=0, bClamp=False, Value=@n6.Value.X)
   > &RigUnit_ModifyTransforms(
    ItemToModify=((Item=(Type=Bone,Name="Dr_HairTailMain"),Transform=(Rotation=(X=0,Y=0,Z=0,W=1),Translation=(X=0,Y=2,Z=-3),Scale3D=(X=1,Y=1,Z=1))))
    WeightMinimum=0
    Weight=@n7.Result
   )
   > &RigUnit_ModifyTransforms(
    ItemToModify=((Item=(Type=Bone,Name="Dr_HairTailMain"),Transform=(Rotation=(X=-0.2588,Y=0,Z=0,W=0.9659),Translation=(X=0,Y=-2,Z=3),Scale3D=(X=1,Y=1,Z=1))))
    WeightMinimum=0
    Weight=&RigVMFunction_MathDoubleRemap(SourceMaximum=-60, SourceMinimum=0, TargetMaximum=1, TargetMinimum=0, bClamp=False, Value=@n6.Value.X).Result
   )
   > &RigUnit_ModifyTransforms(
    ItemToModify="((Item=(Type=Bone,Name=\"Dr_BangsHair\"),Transform=(Rotation=(X=-0.1522,Y=-0.0467,Z=-0.2998,W=0.9406),Translation=(X=1.4023,Y=0.0725,Z=0.3538),Scale3D=(X=1,Y=1,Z=1))),(Item=(Type=Bone,Name=\"Dr_HairClaviclr_l\"),Transform=(Rotation=(X=0,Y=0,Z=0,W=1),Translatio ...(truncated)"
    WeightMinimum=0
    Weight=@n7.Result
   )
   > &RigUnit_ModifyTransforms(
    ItemToModify=((Item=(Type=Bone,Name="dr_backHat"),Transform=(Rotation=(X=-0.4067,Y=0,Z=0,W=0.9135),Translation=(X=0,Y=0,Z=0),Scale3D=(X=1,Y=1,Z=1))))
    WeightMinimum=0
    Weight=&RigVMFunction_MathDoubleRemap(SourceMaximum=-98, SourceMinimum=-60, TargetMaximum=1, TargetMinimum=0, bClamp=False, Value=RigVMVariableNode() {variable=HeadRot}.Value.Z).Result
   )
   n8 = &RigUnit_GetTransform(Item=(Type=Bone,Name="clavicle_l"), bInitial=False)
   &RigUnit_ModifyTransforms(
    ItemToModify="((Item=(Type=Bone,Name=\"Dr_HairClaviclr_l\"),Transform=(Rotation=(X=0.2094,Y=-0.198,Z=0.4319,W=0.8546),Translation=(X=-1.0948,Y=0.1158,Z=7.5309),Scale3D=(X=1,Y=1,Z=1))),(Item=(Type=Bone,Name=\"Dr_BangsHair\"),Transform=(Rotation=(X=-0.0304,Y=-0.1014,Z=0.0067, ...(truncated)"
    WeightMinimum=0
    Weight=&RigVMFunction_MathDoubleRemap(
     SourceMaximum=-40
     SourceMinimum=2
     TargetMaximum=1
     TargetMinimum=0
     bClamp=False
     Value=&RigVMFunction_MathQuaternionToEuler(RotationOrder=ZYX, Value=@n8.Transform.Rotation).Result.Y
    ).Result
    ExecutePin=@RigVMFunction_Sequence_2.D
   )
   RigUnit_GetTransform = &RigUnit_GetTransform(Item=(Type=Bone,Name="clavicle_r"), Space=LocalSpace, bInitial=False)
   > &RigUnit_ModifyTransforms(
    ItemToModify=((Item=(Type=Bone,Name="Dr_HairClaviclr_r"),Transform=(Rotation=(X=0.1379,Y=0.0169,Z=0.0201,W=0.9901),Translation=(X=1.191,Y=-1.6416,Z=3.8671),Scale3D=(X=1,Y=1,Z=1))))
    WeightMinimum=0
    Weight=&RigVMFunction_MathDoubleRemap(
     SourceMaximum=-50
     SourceMinimum=-80
     TargetMaximum=1
     TargetMinimum=0
     bClamp=False
     Value=&RigVMFunction_MathQuaternionToEuler(RotationOrder=ZYX, Value=@RigUnit_GetTransform.Transform.Rotation).Result.Y
    ).Result
   )
   RigVMFunction_ControlFlowBranch = &RigVMFunction_ControlFlowBranch(ExecuteContext=@RigVMFunction_Sequence_2.E, Condition=RigVMVariableNode() {variable=IsWearingBackbling}.Value)
   &RigUnit_ParentConstraint(
    AdvancedSettings=(InterpolationType=Average,RotationOrderForFilter=XZY)
    Child=(Type=Bone,Name="Dyn_HairTail2")
    Parents=((Item=(Type=Bone,Name="spine_05"),Weight=1))
    Weight=1
    ExecutePin=@RigVMFunction_ControlFlowBranch.True
   )
   > &RigUnit_ParentConstraint(
    AdvancedSettings=(InterpolationType=Average,RotationOrderForFilter=XZY)
    Child=(Type=Bone,Name="Dyn_HairTail3")
    Parents=((Item=(Type=Bone,Name="spine_03"),Weight=1))
    Weight=1
   )
   > &RigUnit_ModifyTransforms(
    ItemToModify="((Item=(Type=Bone,Name=\"Dyn_HairTail2\"),Transform=(Rotation=(X=0,Y=0,Z=0.0784,W=0.9969),Translation=(X=0,Y=0,Z=0),Scale3D=(X=1,Y=1,Z=1))),(Item=(Type=Bone,Name=\"Dyn_HairTail3\"),Transform=(Rotation=(X=0,Y=0,Z=0.028,W=0.9996),Translation=(X=0,Y=0,Z=0),Scale3 ...(truncated)"
    Weight=1
    WeightMinimum=0
   )
   > n9 = &RigUnit_SphericalPoseReader(
    ActiveRegionScaleFactors=(PositiveWidth=1,NegativeWidth=1,PositiveHeight=1,NegativeHeight=1)
    ActiveRegionSize=0.1
    Debug=(bDrawDebug=True,bDraw2D=False,bDrawLocalAxes=False,DebugScale=25,DebugSegments=20,DebugThickness=0.25)
    DriverAxis=(X=0,Y=1,Z=0)
    DriverItem=(Type=Bone,Name="head")
    FalloffRegionScaleFactors=(PositiveWidth=1,NegativeWidth=1,PositiveHeight=1,NegativeHeight=1)
    FalloffSize=0.25
    FlipHeightScaling=False
    FlipWidthScaling=False
    OptionalParentItem=(Type=Bone,Name="pelvis")
    RotationOffset=(X=90,Y=0,Z=-80)
   )
   n10 = &RigVMFunction_MathVectorDistance(
    A=&RigUnit_GetTransform(Item=(Type=Bone,Name="root"), bInitial=False).Transform.Translation
    B=&RigUnit_GetTransform(Item=(Type=Bone,Name="pelvis"), bInitial=False).Transform.Translation
   )
   > &RigUnit_ModifyTransforms(
    ItemToModify=((Item=(Type=Bone,Name="Dr_HairTail1"),Transform=(Rotation=(X=0,Y=0,Z=0,W=1),Translation=(X=0.5,Y=0.5,Z=0),Scale3D=(X=1,Y=1,Z=1))))
    WeightMinimum=0
    Weight=&RigVMFunction_MathDoubleMul(
     B=&RigVMFunction_MathDoubleRemap(SourceMaximum=80, SourceMinimum=90, TargetMaximum=1, TargetMinimum=0, Value=@n10.Result).Result
     A=@n9.OutputParam
    ).Result
   )
   n11 = &RigVMFunction_ControlFlowBranch(ExecuteContext=@RigVMFunction_Sequence_2.F, Condition=RigVMVariableNode() {variable=RBAN}.Value)
   &RigUnit_ParentConstraint(
    AdvancedSettings=(InterpolationType=Average,RotationOrderForFilter=XZY)
    Child=(Type=Bone,Name="Dyn_HairTail2")
    Parents=((Item=(Type=Bone,Name="spine_05"),Weight=1))
    Weight=1
    ExecutePin=@n11.False
   )
   > &RigUnit_ParentConstraint(
    AdvancedSettings=(InterpolationType=Average,RotationOrderForFilter=XZY)
    Child=(Type=Bone,Name="Dyn_HairTail3")
    Parents=((Item=(Type=Bone,Name="spine_04"),Weight=1))
    Weight=1
   )
   > &RigUnit_ModifyTransforms(
    ItemToModify=((Item=(Type=Bone,Name="Dyn_HairTail3"),Transform=(Rotation=(X=0,Y=0,Z=0.1845,W=0.9828),Translation=(X=0,Y=0,Z=0),Scale3D=(X=1,Y=1,Z=1))))
    Weight=1
    WeightMinimum=0
   )
   > &RigUnit_ParentConstraint(
    AdvancedSettings=(InterpolationType=Average,RotationOrderForFilter=XZY)
    Child=(Type=Bone,Name="Dyn_HairTail4")
    Parents=((Item=(Type=Bone,Name="spine_03"),Weight=1))
    Weight=1
   )
   > &RigUnit_ModifyTransforms(
    ItemToModify=((Item=(Type=Bone,Name="Dyn_HairTail4"),Transform=(Rotation=(X=0,Y=0,Z=-0.013,W=0.9999),Translation=(X=0.3519,Y=4.5654,Z=0),Scale3D=(X=1,Y=1,Z=1))))
    Weight=1
    WeightMinimum=0
   )
   > &RigUnit_RotationConstraintLocalSpaceOffset(
    AdvancedSettings=(InterpolationType=Average,RotationOrderForFilter=XZY)
    Child=(Type=Bone,Name="Dyn_HairTail5")
    Filter=(bX=True,bY=True,bZ=True)
    Parents=((Item=(Type=Bone,Name="spine_01"),Weight=1))
    Weight=1
   )
   > &RigUnit_ModifyTransforms(
    ItemToModify=((Item=(Type=Bone,Name="Dyn_HairTail5"),Transform=(Rotation=(X=0,Y=0,Z=-0.2685,W=0.9633),Translation=(X=0,Y=0,Z=0),Scale3D=(X=1,Y=1,Z=1))))
    Weight=1
    WeightMinimum=0
   )
   Maximum = &RigVMFunction_MathDoubleMax(A=@n4.Result, B=@n5.Result)
  ]}
]
hierarchy [
 root {loc=<0 0 0>, rot=<0 0 0 1>}
 pelvis {parent=root, loc=<0 -0.5301 98.0571>, rot=<0.0127 0.707 -0.0127 -0.707>}
 spine_01 {parent=pelvis, loc=<6.1208 -0.9328 0>, rot=<0 0 0.034 -0.9994>}
 spine_02 {parent=spine_01, loc=<6.5996 -0.2095 0>, rot=<0 0 0.0238 -0.9997>}
 spine_03 {parent=spine_02, loc=<6.4914 -0.1237 0>, rot=<0 0 0.0149 -0.9999>}
 spine_04 {parent=spine_03, loc=<6.9598 0.0794 0>, rot=<0 0 -0.0132 -0.9999>}
 spine_05 {parent=spine_04, loc=<6.9987 0 0>, rot=<0 0 -0.0015 -1>}
 neck_01 {parent=spine_05, loc=<11.3138 0.1529 0>, rot=<0 0 -0.2003 -0.9797>}
 neck_02 {parent=neck_01, loc=<4.5627 0 0>, rot=<0 0 0.0504 -0.9987>}
 head {parent=neck_02, loc=<5.3914 0 0>, rot=<0 0 0.115 -0.9934>}
 HeadAccRoot {parent=head, loc=<0 0 0>, rot=<0 0 0 1>}
 spectacles {parent=HeadAccRoot, loc=<4.0695 -0.3353 0>, rot=<-0.601 -0.7992 0 0>}
 spectaclesEnd {parent=spectacles, loc=<9.8089 0 0>, rot=<0 0 0 1>}
 earring_l {parent=HeadAccRoot, loc=<0.0695 0.5117 -6.3155>, rot=<0.9998 0.0214 0 0>}
 earring_r {parent=HeadAccRoot, loc=<0.0692 0.5116 6.3156>, rot=<0.0214 -0.9998 0 0>}
 earring_End_r {parent=earring_r, loc=<5.3347 0 0>, rot=<0 0 0 1>}
 Dr_HairTailRoot {parent=HeadAccRoot, loc=<5.9697 -6.1303 0>, rot=<0.1776 -0.9841 0 0>}
 Dr_HairClaviclr_l {parent=Dr_HairTailRoot, loc=<14.179 5.6174 8.7938>, rot=<-0.1227 0.6964 -0.1227 -0.6964>}
 Dr_HairClaviclr_r {parent=Dr_HairTailRoot, loc=<12.5515 5.9331 -10.3315>, rot=<-0.1227 0.6964 -0.1227 -0.6964>}
 Dr_HairStradClavicle_r {parent=Dr_HairClaviclr_r, loc=<0.8251 -6.6928 -6.5288>, rot=<0 0 0 1>}
 Dyn_HairStradClavicle_r {parent=Dr_HairStradClavicle_r, loc=<0 0 0>, rot=<0 0 0 1>}
 Dyn_HairStrand_r {parent=Dr_HairClaviclr_r, loc=<0 0 0>, rot=<0 0 0 1>}
 Dr_HairTailMain {parent=Dr_HairTailRoot, loc=<10.5806 0 0>, rot=<0 0 -0.0889 -0.996>}
 Dr_HairTail1 {parent=Dr_HairTailMain, loc=<0 0 0>, rot=<0 0 0 1>}
 Dyn_HairTail2 {parent=Dr_HairTail1, loc=<8.0973 0.2438 0>, rot=<0 0 -0.0509 -0.9987>}
 Dyn_HairTail3 {parent=Dyn_HairTail2, loc=<8.7447 -0.5737 0>, rot=<0 0 0.0743 -0.9972>}
 Dyn_HairTail4 {parent=Dyn_HairTail3, loc=<10.6141 1.004 0>, rot=<0 0 -0.2139 -0.9769>}
 Dyn_HairTail5 {parent=Dyn_HairTail4, loc=<7.932 -2.0892 0>, rot=<0 0 -0.0324 -0.9995>}
 Dyn_HairStradBack {parent=Dyn_HairTail3, loc=<10.6349 0.5262 -8.2289>, rot=<-0.0768 0.7029 -0.0768 -0.7029>}
 Dr_BangsHair {parent=HeadAccRoot, loc=<4.4643 7.3655 -6.5017>, rot=<0.0029 -0.7071 -0.0029 -0.7071>}
 Dyn_BangsHair_1 {parent=Dr_BangsHair, loc=<2.5378 -0.5808 -5.2954>, rot=<0 0 0 1>}
 Dyn_BangsHair_2 {parent=Dyn_BangsHair_1, loc=<1.697 -2.0436 -4.4748>, rot=<0 0 0 1>}
 Dyn_HairStrand_l {parent=Dr_BangsHair, loc=<3.719 -4.2597 -3.7554>, rot=<0 0 0 1>}
 HairStrandEnd_l {parent=Dyn_HairStrand_l, loc=<3.6377 -4.3042 -5.9745>, rot=<0 0 0 1>}
 cap {parent=HeadAccRoot, loc=<9.5393 0.0788 0>, rot=<0.0029 -0.7071 -0.0029 -0.7071>}
 Dyn_hat_front_r {parent=cap, loc=<-2.9792 4.4715 0.6806>, rot=<-0.13 0.013 0.9475 0.2918>}
 Dyn_hat_front_l {parent=cap, loc=<2.1575 4.8041 -0.0745>, rot=<0.0017 0.0031 -0.485 -0.8745>}
 Dyn_hat_back_l {parent=cap, loc=<3.3735 -1.7468 0.0524>, rot=<0.0458 -0.1408 0.3062 -0.9404>}
 Dyn_hat_back_r {parent=cap, loc=<-2.9379 -2.0382 -0.3909>, rot=<-0.0352 -0.0171 0.9146 -0.4025>}
 dr_backHat {parent=cap, loc=<-0.0002 -3.3844 -0.2608>, rot=<-0.0872 0 0 -0.9962>}
 clavicle_l {parent=spine_05, loc=<5.4831 -0.1053 -3.2946>, rot=<0.0465 -0.7289 -0.009 -0.6829>}
 upperarm_l {parent=clavicle_l, loc=<9.8658 -1.2136 -0.4105>, rot=<0.0164 -0.3245 -0.0269 -0.9454>}
 clavicle_r {parent=spine_05, loc=<5.4831 -0.1053 3.2946>, rot=<0.7289 0.0465 -0.6829 0.009>}
 upperarm_r {parent=clavicle_r, loc=<-9.8658 1.2136 0.4105>, rot=<0.0164 -0.3245 -0.0269 -0.9454>}
]
variables [
 HeadRot {type=FVector, default=(X=0,Y=0,Z=0)}
 IsWearingBackbling {type=bool, default=False}
 RBAN {type=bool, default=False}
]
