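# Physics asset for the FabulousWind female character's face accessory mesh, set up for hair physics on the ponytail, bangs and strands.
# Limited-swing constraints chain Dyn_ hair bones to Dr_ bones, with tapering boxes along the tail and the back strand branching from HairTail3.
# Spheres and capsules on head, neck, spine, pelvis, clavicles and upper arms provide body collision shapes, unconstrained except a free left upperarm-to-clavicle joint.
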PhysicsAsset /BRCosmetics/Characters/Player/Female/Medium/Bodies/F_MED_FabulousWind/Meshes/Parts/F_MED_FabulousWind_FaceAcc_Physics
bodies {
 Dr_BangsHair [{shapeType=Sphere, radius=1.1949, shapeName=Dr_BangsHair_sphere}]
 Dr_HairClaviclr_l [{shapeType=Sphere, radius=1.499, shapeName=Dr_HairClaviclr_l_sphere}]
 Dr_HairClaviclr_r [{shapeType=Sphere, radius=0.9982, shapeName=Dr_HairClaviclr_r_sphere}]
 Dr_HairTail1 [{shapeType=Capsule, length=2.0903, radius=3.2226, shapeName=Dr_HairTail1_capsule}]
 Dr_HairTailMain [{shapeType=Sphere, radius=1.3682, shapeName=Dr_HairTailMain_sphere}]
 Dr_HairTailRoot [{shapeType=Sphere, radius=1.6495, shapeName=Dr_HairTailRoot_sphere}]
 Dyn_BangsHair_1 [{
   shapeType=Capsule
   length=3.9669
   radius=1.537
   shapeName=Dyn_BangsHair_1_capsule
   center=<0.7364 -0.7837 -2.0233>
   rotation=<rot 20 -0.0002 19.9995>
  }]
 Dyn_BangsHair_2 [{
   shapeType=Capsule
   length=3.2995
   radius=2.4405
   shapeName=Dyn_BangsHair_2_capsule
   center=<-0.4305 -2.7986 -2.1995>
   rotation=<rot -8.8765 -5.2328 50.1031>
  }]
 Dyn_HairStradBack [{
   shapeType=Capsule
   length=6.9652
   radius=1.9553
   shapeName=Dyn_HairStradBack_capsule
   center=<0.9811 0 -4.9551>
   rotation=<rot 17.5821 0 -0.0001>
  }]
 Dyn_HairStrand_r [{
   shapeType=Capsule
   length=2.9819
   radius=1.0615
   shapeName=Dyn_HairStrand_r_capsule
   center=<-1.2921 -2.5322 -2.1248>
   rotation=<rot -12.7002 -15.5793 51.7438>
  }]
 Dyn_HairTail2 [{
   shapeType=Box
   extentX=13.2634
   extentY=3.1405
   extentZ=13.0253
   shapeName=Dyn_HairTail2_box
   center=<4.8597 -0.3789 -0.0033>
   rotation=<rot 0 4.8957 0>
  }]
 Dyn_HairTail3 [{
   shapeType=Box
   extentX=7.5
   extentY=2.5144
   extentZ=12.4989
   shapeName=Dyn_HairTail3_box
   center=<6.1668 1.1118 0>
   rotation=<rot 0 3.9301 0>
  }]
 Dyn_HairTail4 [{
   shapeType=Box
   extentX=7.5
   extentY=3.2599
   extentZ=8.54
   shapeName=Dyn_HairTail4_box
   center=<4.0421 -1.2972 0>
   rotation=<rot 0 -12.6827 0>
  }]
 Dyn_HairTail5 [{shapeType=Box, extentX=8.4972, extentY=3.0259, extentZ=5.9203, shapeName=Dyn_HairTail5_box, center=<5.9122 0 0>}]
 clavicle_l [
  {shapeType=Sphere, radius=7.2709, shapeName=clavicle_r_sphere, center=<11.4883 0 0>}
  {shapeType=Capsule, length=8.334, radius=4.8772, shapeName=clavicle_r_capsule, center=<5.2934 0 0>, rotation=<rot -90 -179.156 -0.8443>}
 ]
 clavicle_r [
  {shapeType=Sphere, radius=7.2709, shapeName=clavicle_r_sphere, center=<-11.4883 0 0>}
  {shapeType=Capsule, length=8.334, radius=4.8772, shapeName=clavicle_r_capsule, center=<-5.2934 0 0>, rotation=<rot -90 -174.9489 174.9486>}
 ]
 head [
  {shapeType=Capsule, length=2.6976, radius=6.8922, shapeName=head_capsule, center=<4.6892 1.004 0>, rotation=<rot 0 -170.0017 -90>}
  {shapeType=Capsule, length=4.6918, radius=6.4095, shapeName=head_capsule, center=<2.7484 3.1517 0>, rotation=<rot -90 0.0003 -0.0006>}
 ]
 neck_01 [{shapeType=Capsule, length=7.0383, radius=4.4847, shapeName=neck_01_capsule, center=<3.2408 0.191 0>, rotation=<rot -90 -179.9605 179.9601>}]
 pelvis [{shapeType=Capsule, length=7.0249, radius=11.2938, shapeName=pelvis_capsule, center=<0 -0.3853 0>}]
 spine_01 [{shapeType=Capsule, length=10.4736, radius=7.8002, shapeName=spine_01_capsule, center=<0 0.409 0>}]
 spine_02 [{shapeType=Capsule, length=6.364, radius=5.9539, shapeName=spine_03_capsule, center=<3.6864 1.7392 0>}]
 spine_03 [{shapeType=Capsule, length=8.4432, radius=5.9539, shapeName=spine_03_capsule, center=<3.6864 0.7567 0>}]
 spine_05 [{shapeType=Capsule, length=11.749, radius=8.4153, shapeName=spine_05_capsule, center=<1.0379 0.5217 0>}]
 upperarm_l [{shapeType=Capsule, length=21.5907, radius=4.9985, shapeName=upperarm_r_capsule, center=<12.3439 0 0>, rotation=<rot -90 0 180>}]
 upperarm_r [{shapeType=Capsule, length=21.5907, radius=4.9985, shapeName=upperarm_r_capsule, center=<-12.3439 0 0>, rotation=<rot -90 0 0>}]
}
constraints [
 {bone1Name=Dyn_HairTail5, bone2Name=Dyn_HairTail4, swing1LimitDegrees=58.04, swing2LimitDegrees=5, twistLimitDegrees=5}
 {bone1Name=Dyn_HairTail4, bone2Name=Dyn_HairTail3, swing1LimitDegrees=60, swing2LimitDegrees=30, twistLimitDegrees=10}
 {bone1Name=Dyn_HairTail3, bone2Name=Dyn_HairTail2, swing1LimitDegrees=50, swing2LimitDegrees=30, twistLimitDegrees=10}
 {bone1Name=Dyn_HairTail2, bone2Name=Dr_HairTail1, swing1LimitDegrees=50, swing2LimitDegrees=30, twistLimitDegrees=30}
 {bone1Name=Dyn_BangsHair_2, bone2Name=Dyn_BangsHair_1, swing1LimitDegrees=10, swing2LimitDegrees=10, twistLimitDegrees=5}
 {bone1Name=Dyn_BangsHair_1, bone2Name=Dr_BangsHair, swing1LimitDegrees=10, swing2LimitDegrees=10, twistLimitDegrees=5}
 {bone1Name=Dyn_HairStrand_r, bone2Name=Dr_HairClaviclr_r, swing1LimitDegrees=15, swing2LimitDegrees=15, twistLimitDegrees=5}
 {bone1Name=Dr_HairTail1, bone2Name=Dr_HairTailMain, swing1LimitDegrees=35, swing2LimitDegrees=30, twistLimitDegrees=10}
 {bone1Name=Dyn_HairStradBack, bone2Name=Dyn_HairTail3, swing1LimitDegrees=15, swing2LimitDegrees=15, twistLimitDegrees=15}
 {
  bone1Name=upperarm_l
  bone2Name=clavicle_l
  swing1LimitDegrees=45
  swing1Motion=Free
  swing2LimitDegrees=45
  swing2Motion=Free
  twistLimitDegrees=45
  twistMotion=Free
 }
]
