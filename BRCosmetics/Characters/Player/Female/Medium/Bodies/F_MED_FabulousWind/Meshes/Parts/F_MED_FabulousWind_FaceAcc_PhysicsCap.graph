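# Physics asset for the FabulousWind female medium cosmetic's face accessory, giving its hat pieces and earrings constrained secondary motion around the head.
# Two head capsules form the central body, with a sphere on each of four dynamic hat bones and a capsule on each earring.
# Every part is constrained directly to the head with twist locked.
# Hat bones swing on a single narrow axis, while earrings swing within wider limits on both axes.
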PhysicsAsset /BRCosmetics/Characters/Player/Female/Medium/Bodies/F_MED_FabulousWind/Meshes/Parts/F_MED_FabulousWind_FaceAcc_PhysicsCap
bodies {
 Dyn_hat_back_l [{shapeType=Sphere, radius=1.827, shapeName=Dyn_hat_front_l_sphere, center=<12.5289 0 0>}]
 Dyn_hat_back_r [{shapeType=Sphere, radius=1.827, shapeName=Dyn_hat_front_l_sphere, center=<12.5289 0 0>}]
 Dyn_hat_front_l [{shapeType=Sphere, radius=1.827, shapeName=Dyn_hat_front_l_sphere, center=<12.5289 0 0>}]
 Dyn_hat_front_r [{shapeType=Sphere, radius=1.827, shapeName=Dyn_hat_front_l_sphere, center=<12.5289 0 0>}]
 earring_l [{shapeType=Capsule, length=2.7892, radius=0.7098, shapeName=earring_r_capsule, center=<-3.6173 0 0>, rotation=<rot 90 180 -0.0005>}]
 earring_r [{shapeType=Capsule, length=2.7892, radius=0.7098, shapeName=earring_r_capsule, center=<3.6173 0 0>, rotation=<rot 90 180 179.9995>}]
 head [
  {shapeType=Capsule, length=2.6976, radius=6.8922, shapeName=head_capsule, center=<4.6892 1.004 0>, rotation=<rot 0 -170.0017 -90>}
  {shapeType=Capsule, length=4.6918, radius=6.4095, shapeName=head_capsule, center=<2.7484 3.1517 0>, rotation=<rot -90 0.0003 -0.0006>}
 ]
}
constraints [
 {
  bone1Name=Dyn_hat_front_l
  bone2Name=head
  swing1LimitDegrees=45
  swing1Motion=Locked
  swing2LimitDegrees=15
  twistLimitDegrees=45
  twistMotion=Locked
 }
 {
  bone1Name=Dyn_hat_front_r
  bone2Name=head
  swing1LimitDegrees=45
  swing1Motion=Locked
  swing2LimitDegrees=15
  twistLimitDegrees=45
  twistMotion=Locked
 }
 {
  bone1Name=Dyn_hat_back_r
  bone2Name=head
  swing1LimitDegrees=45
  swing1Motion=Locked
  swing2LimitDegrees=15
  twistLimitDegrees=45
  twistMotion=Locked
 }
 {
  bone1Name=Dyn_hat_back_l
  bone2Name=head
  swing1LimitDegrees=45
  swing1Motion=Locked
  swing2LimitDegrees=15
  twistLimitDegrees=45
  twistMotion=Locked
 }
 {bone1Name=earring_r, bone2Name=head, swing1LimitDegrees=35, swing2LimitDegrees=35, twistLimitDegrees=45, twistMotion=Locked}
 {bone1Name=earring_l, bone2Name=head, swing1LimitDegrees=35, swing2LimitDegrees=35, twistLimitDegrees=45, twistMotion=Locked}
]
