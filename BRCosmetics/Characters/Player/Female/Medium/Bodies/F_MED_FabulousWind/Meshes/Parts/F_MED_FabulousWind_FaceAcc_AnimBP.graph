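# Anim Blueprint for the FabulousWind female face accessory, adding secondary motion to hair and cap parts.
# It copies the parent pose through a Control Rig and a head-space RigidBody, then caches it.
# Bool blends choose between three spine_05 RigidBody branches: state-driven gravity and wind override, sine-noise hair tail, or fixed gravity.
# The event graph computes that override and scales phase-offset tail noise by pawn speed.
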AnimBlueprint /BRCosmetics/Characters/Player/Female/Medium/Bodies/F_MED_FabulousWind/Meshes/Parts/F_MED_FabulousWind_FaceAcc_AnimBP
target_skeleton=&/F_MED_FabulousWind_FaceAcc_Skeleton
graphs [
 AnimGraph {nodes [
   RigidBody!(ComponentPose, Alpha, OverrideWorldGravity:Vector="0,0,-980", ExternalForce) -> (Pose) {
    bOverrideWorldGravity=true
    overridePhysicsAsset=&/F_MED_FabulousWind_FaceAcc_Physics
    simulationSpace=BaseBoneSpace
    alphaBoolBlend {blendOption=Linear}
    baseBoneRef {boneName=spine_05}
    componentAppliedLinearAccClamp=<1000 1000 1000>
    componentLinearAccScale=<0.15 0.15 0.15>
    componentLinearVelScale=<1 1 1>
    overrideWorldGravity {z=-980}
    simSpaceSettings {maxAngularAcceleration=60, maxAngularVelocity=3, maxLinearAcceleration=80, maxLinearVelocity=5, velocityScaleZ=0, worldAlpha=1}
   }
   RigidBody(OverrideWorldGravity="0,0,-980") {
    bOverrideWorldGravity=true
    overridePhysicsAsset=&F_MED_FabulousWind_FaceAcc_Physics
    simulationSpace=BaseBoneSpace
    alphaBoolBlend {blendOption=Linear}
    baseBoneRef {boneName=spine_05}
    componentAppliedLinearAccClamp=<1000 1000 1000>
    componentLinearAccScale=<0.15 0.15 0.15>
    componentLinearVelScale=<1 1 1>
    overrideWorldGravity {z=-980}
    simSpaceSettings {maxAngularAcceleration=60, maxAngularVelocity=3, maxLinearAcceleration=80, maxLinearVelocity=5, velocityScaleZ=0, worldAlpha=1}
   }
   n1 = RigidBody(ComponentPose=Local To Component(LocalPose=Use cached pose 'PoseCachStageOne'()), OverrideWorldGravity="0,0,0") {
    bOverrideWorldGravity=true
    overridePhysicsAsset=&F_MED_FabulousWind_FaceAcc_Physics
    simulationSpace=BaseBoneSpace
    alphaBoolBlend {blendOption=Linear}
    baseBoneRef {boneName=spine_05}
    componentAppliedLinearAccClamp=<2000 2000 2000>
    simSpaceSettings {maxAngularAcceleration=50, maxAngularVelocity=5, maxLinearAcceleration=150, maxLinearVelocity=25, velocityScaleZ=0, worldAlpha=0.5}
   }
   n2 = "Transform (Modify) Bone - Bone: Dyn_HairTail5"(ComponentPose=@n1, Alpha=1, Translation="0,0,0", Rotation=Get NoisRot4() -> (NoisRot4:Rotator), Scale="1,1,1") {rotationMode=BMM_Additive, rotationSpace=BCS_BoneSpace, alphaBoolBlend {blendOption=Linear}, boneToModify {boneName=Dyn_HairTail5}}
   n3 = "Transform (Modify) Bone - Bone: Dyn_HairTail4"(ComponentPose=@n2, Alpha=1, Translation="0,0,0", Rotation=Get NoisRot3() -> (NoisRot3:Rotator), Scale="1,1,1") {rotationMode=BMM_Additive, rotationSpace=BCS_BoneSpace, alphaBoolBlend {blendOption=Linear}, boneToModify {boneName=Dyn_HairTail4}}
   n4 = "Transform (Modify) Bone - Bone: Dyn_HairTail3"(ComponentPose=@n3, Alpha=1, Translation="0,0,0", Rotation=Get NoisRot2() -> (NoisRot2:Rotator), Scale="1,1,1") {rotationMode=BMM_Additive, rotationSpace=BCS_BoneSpace, alphaBoolBlend {blendOption=Linear}, boneToModify {boneName=Dyn_HairTail3}}
   n5 = "Transform (Modify) Bone - Bone: Dyn_HairTail2"(ComponentPose=@n4, Alpha=1, Translation="0,0,0", Rotation=Get NoisRot1() -> (NoisRot1:Rotator), Scale="1,1,1") {rotationMode=BMM_Additive, rotationSpace=BCS_BoneSpace, alphaBoolBlend {blendOption=Linear}, boneToModify {boneName=Dyn_HairTail2}}
   n6 = Component To Local(ComponentPose=@n5)
   n7 = RigidBody(ComponentPose=Local To Component(LocalPose=Use cached pose 'PoseCachStageOne'()), OverrideWorldGravity="0,0,-980") {
    bOverrideWorldGravity=true
    overridePhysicsAsset=&F_MED_FabulousWind_FaceAcc_Physics
    simulationSpace=BaseBoneSpace
    alphaBoolBlend {blendOption=Linear}
    baseBoneRef {boneName=spine_05}
    componentAppliedLinearAccClamp=<1000 1000 1000>
    componentLinearAccScale=<0.2 0.2 0.2>
    componentLinearVelScale=<1 1 1>
    overrideWorldGravity {z=-980}
    simSpaceSettings {maxAngularAcceleration=300, maxAngularVelocity=3, maxLinearAcceleration=80, maxLinearVelocity=5, worldAlpha=1}
   }
   n8 = Component To Local(ComponentPose=@n7)
   n9 = RigidBody(ComponentPose=Local To Component(LocalPose=Use cached pose 'PoseCachStageOne'()), OverrideWorldGravity=Get Gravity Override()) {
    bOverrideWorldGravity=true
    overridePhysicsAsset=&/F_MED_FabulousWind_FaceAcc_Physics1
    simulationSpace=BaseBoneSpace
    alphaBoolBlend {blendOption=Linear}
    baseBoneRef {boneName=spine_05}
    componentAppliedLinearAccClamp=<150 150 150>
    componentLinearAccScale=<1 1 1>
    componentLinearVelScale=<1 1 1>
    simSpaceSettings {maxAngularAcceleration=50, maxAngularVelocity=25, maxLinearAcceleration=50, maxLinearVelocity=25, velocityScaleZ=0, worldAlpha=0.4}
   }
   Output Pose(Result=Blend Poses by bool(BlendPose_0=Component To Local(ComponentPose=@n9), BlendPose_1=Blend Poses by bool(BlendPose_0=@n6, BlendPose_1=@n8) {blendType=Linear}) {blendType=Linear}) {layerGroup=None}
   n10 = Control Rig!(
    Source=Copy Pose From Mesh() {bUseAttachedParent=true}
    Alpha
    IsWearingBackbling:bool=Get IsWearingBackbling() -> (IsWearingBackbling:bool)
    RBAN:bool=Get bRigidBodyEnabled()
   ) -> (Pose) {
    alphaBoolBlend {blendOption=Linear}
    controlRigAssetReference {blueprintRigClass=&/F_MED_FabulousWind_FaceAcc_CtrlRig.F_MED_FabulousWind_FaceAcc_CtrlRig_C}
   }
   n11 = RigidBody!(ComponentPose=Local To Component(LocalPose=@n10), Alpha, ExternalForce) -> (Pose) {
    overridePhysicsAsset=&/F_MED_FabulousWind_FaceAcc_PhysicsCap
    simulationSpace=BaseBoneSpace
    alphaBoolBlend {blendOption=Linear}
    baseBoneRef {boneName=head}
    componentAppliedLinearAccClamp=<150 150 150>
    componentLinearAccScale=<1 1 1>
    componentLinearVelScale=<1 1 1>
    simSpaceSettings {maxAngularAcceleration=50, maxAngularVelocity=25, maxLinearAcceleration=50, maxLinearVelocity=25, velocityScaleZ=0, worldAlpha=0.2}
   }
   Save cached pose 'PoseCachStageOne'(Pose=Component To Local(ComponentPose=@n11)) {cacheName=PoseCachStageOne}
  ]}
 EventGraph {nodes [
   Event BlueprintUpdateAnimation()
   > n1 = Gravity Override State Function DEPRECATED(
    CharacterPartAnimInstance=Self-Reference()
    GameTimeInSeconds=GetGameTimeInSeconds()
    GravityOverrideParams=Make Gravity Override Params Struct!(
     Skydive_GoingUp_28_#1:Gravity Override Setting Struct Structure=Make Gravity Override Setting Struct!(
      WindFrequency_28_#2:Vector="50,20,40"
      WindAmplitude_31_#3:Vector="150,100,50"
      WindOffset_32_#4:Vector="0,0,0"
      GravityOverride_15_#5:Vector="0,0,-100"
      JointName_27_#6:Name=root
     ) -> (GravityOverrideSettingStruct:Gravity Override Setting Struct Structure) {is=title K2Node_MakeStruct}
     Skydive_FloatingDown_29_#7:Gravity Override Setting Struct Structure=Make Gravity Override Setting Struct(
      WindFrequency_28_#2="55,22,40"
      WindAmplitude_31_#3="300,-150,50"
      WindOffset_32_#4="0,0,0"
      GravityOverride_15_#5="0,-100,100"
      JointName_27_#6=root
     )
     Skydive_DivingDown_30_#8:Gravity Override Setting Struct Structure=Make Gravity Override Setting Struct(
      WindFrequency_28_#2="3,4,5"
      WindAmplitude_31_#3="150,200,60"
      WindOffset_32_#4="0,0,0"
      GravityOverride_15_#5="0,-300,600"
      JointName_27_#6=root
     )
     Skydive_Parachuting_31_#9:Gravity Override Setting Struct Structure=Make Gravity Override Setting Struct(
      WindFrequency_28_#2="1.5,7,8"
      WindAmplitude_31_#3="100,-200,200"
      WindOffset_32_#4="0,0,0"
      GravityOverride_15_#5="0,-300,-980"
      JointName_27_#6=root
     )
     OnGround_Standing_32_#10:Gravity Override Setting Struct Structure=(WindFrequency_28_#2=(X=0,Y=0,Z=0),WindAmplitude_31_#3=(X=0,Y=0,Z=0),WindOffset_32_#4=(X=0,Y=0,Z=0),GravityOverride_15_#5=(X=0,Y=0,Z=-980),JointName_27_#6="root")
     OnGround_Moving_33_#11:Gravity Override Setting Struct Structure=Make Gravity Override Setting Struct(
      WindFrequency_28_#2="0,0,0"
      WindAmplitude_31_#3="0,0,0"
      WindOffset_32_#4="0,0,0"
      GravityOverride_15_#5="0,0,-980"
      JointName_27_#6=root
     )
     EmoteOrMelee_34_#12:Gravity Override Setting Struct Structure=(WindFrequency_28_#2=(X=0,Y=0,Z=0),WindAmplitude_31_#3=(X=0,Y=0,Z=0),WindOffset_32_#4=(X=0,Y=0,Z=0),GravityOverride_15_#5=(X=0,Y=0,Z=-980),JointName_27_#6="root")
     Swimming_35_#13:Gravity Override Setting Struct Structure=(WindFrequency_28_#2=(X=0,Y=0,Z=0),WindAmplitude_31_#3=(X=0,Y=0,Z=0),WindOffset_32_#4=(X=0,Y=0,Z=0),GravityOverride_15_#5=(X=-800,Y=0,Z=0),JointName_27_#6="pelvis")
     Front_End_36_#14:Gravity Override Setting Struct Structure=(WindFrequency_28_#2=(X=0,Y=0,Z=0),WindAmplitude_31_#3=(X=0,Y=0,Z=0),WindOffset_32_#4=(X=0,Y=0,Z=0),GravityOverride_15_#5=(X=0,Y=0,Z=-980),JointName_27_#6="root")
     DBNO_40_#15:Gravity Override Setting Struct Structure=(WindFrequency_28_#2=(X=0,Y=0,Z=0),WindAmplitude_31_#3=(X=0,Y=0,Z=0),WindOffset_32_#4=(X=0,Y=0,Z=0),GravityOverride_15_#5=(X=-980,Y=400,Z=0),JointName_27_#6="pelvis")
     DBNO_Carried_41_#16:Gravity Override Setting Struct Structure=(WindFrequency_28_#2=(X=0,Y=0,Z=0),WindAmplitude_31_#3=(X=0,Y=0,Z=0),WindOffset_32_#4=(X=0,Y=0,Z=0),GravityOverride_15_#5=(X=-980,Y=500,Z=-500),JointName_27_#6="pelvis")
     OnGround_Falling_45_#17:Gravity Override Setting Struct Structure=(WindFrequency_28_#2=(X=0,Y=0,Z=0),WindAmplitude_31_#3=(X=0,Y=0,Z=0),WindOffset_32_#4=(X=0,Y=0,Z=0),GravityOverride_15_#5=(X=0,Y=0,Z=-980),JointName_27_#6="root")
     Ziplining_46_#18:Gravity Override Setting Struct Structure=(WindFrequency_28_#2=(X=0,Y=0,Z=0),WindAmplitude_31_#3=(X=0,Y=0,Z=0),WindOffset_32_#4=(X=0,Y=0,Z=0),GravityOverride_15_#5=(X=0,Y=0,Z=-980),JointName_27_#6="root")
     Skydive_Surfing_49_#19:Gravity Override Setting Struct Structure=(WindFrequency_28_#2=(X=1.5,Y=7,Z=8),WindAmplitude_31_#3=(X=200,Y=-500,Z=400),WindOffset_32_#4=(X=0,Y=0,Z=0),GravityOverride_15_#5=(X=500,Y=-300,Z=-980),JointName_27_#6="root")
     Riding_Standing_54_#20:Gravity Override Setting Struct Structure=(WindFrequency_28_#2=(X=0,Y=0,Z=0),WindAmplitude_31_#3=(X=0,Y=0,Z=0),WindOffset_32_#4=(X=0,Y=0,Z=0),GravityOverride_15_#5=(X=0,Y=-350,Z=-980),JointName_27_#6="root")
     Riding_Moving_55_#21:Gravity Override Setting Struct Structure=(WindFrequency_28_#2=(X=0,Y=0,Z=0),WindAmplitude_31_#3=(X=0,Y=0,Z=0),WindOffset_32_#4=(X=0,Y=0,Z=0),GravityOverride_15_#5=(X=0,Y=-450,Z=-450),JointName_27_#6="root")
     Riding_Moving_Targeting_58_#22:Gravity Override Setting Struct Structure=(WindFrequency_28_#2=(X=0,Y=0,Z=0),WindAmplitude_31_#3=(X=0,Y=0,Z=0),WindOffset_32_#4=(X=0,Y=0,Z=0),GravityOverride_15_#5=(X=0,Y=-200,Z=-980),JointName_27_#6="root")
     Tactical_Sprint_57_#23:Gravity Override Setting Struct Structure=Make Gravity Override Setting Struct(
      WindFrequency_28_#2="0,0,0"
      WindAmplitude_31_#3="0,0,0"
      WindOffset_32_#4="0,0,0"
      GravityOverride_15_#5="0,0,-980"
      JointName_27_#6=root
     )
     Rail_Grinding_61_#24:Gravity Override Setting Struct Structure=(WindFrequency_28_#2=(X=0,Y=0,Z=0),WindAmplitude_31_#3=(X=0,Y=0,Z=0),WindOffset_32_#4=(X=0,Y=0,Z=0),GravityOverride_15_#5=(X=0,Y=0,Z=-980),JointName_27_#6="root")
     DBNO_Tumbling_67_#25:Gravity Override Setting Struct Structure=(WindFrequency_28_#2=(X=0,Y=0,Z=0),WindAmplitude_31_#3=(X=0,Y=0,Z=0),WindOffset_32_#4=(X=0,Y=0,Z=0),GravityOverride_15_#5=(X=-2000,Y=0,Z=0),JointName_27_#6="pelvis")
     DBNO_RollFwd_68_#26:Gravity Override Setting Struct Structure=(WindFrequency_28_#2=(X=0,Y=0,Z=0),WindAmplitude_31_#3=(X=0,Y=0,Z=0),WindOffset_32_#4=(X=0,Y=0,Z=0),GravityOverride_15_#5=(X=-2000,Y=0,Z=0),JointName_27_#6="pelvis")
     DBNO_RollBwd_69_#27:Gravity Override Setting Struct Structure=(WindFrequency_28_#2=(X=0,Y=0,Z=0),WindAmplitude_31_#3=(X=0,Y=0,Z=0),WindOffset_32_#4=(X=0,Y=0,Z=0),GravityOverride_15_#5=(X=-4000,Y=0,Z=0),JointName_27_#6="pelvis")
     DBNO_Impulse_70_#28:Gravity Override Setting Struct Structure=(WindFrequency_28_#2=(X=0,Y=0,Z=0),WindAmplitude_31_#3=(X=0,Y=0,Z=0),WindOffset_32_#4=(X=0,Y=0,Z=0),GravityOverride_15_#5=(X=-2000,Y=0,Z=0),JointName_27_#6="pelvis")
    ) -> (GravityOverrideParamsStruct:Gravity Override Params Struct Structure) {is=title K2Node_MakeStruct}
   )
   > Set Gravity Override(Gravity Override=@n1)
   > n2 = Sequence!(execute) -> (then_0, then_1, then_2:Exec, then_3:Exec)
   n3 = MakeVector(X=30, Y=40, Z=30)
   n4 = Remap Float(Input=Vector Length(A=Get PawnVelocity()), OldRange=(X=300,Y=1100), NewRange=(X=0,Y=1), Clamp=true) -> (Output=~)
   > Set NoisRot1(NoisRot1:Rotator="Lerp (Rotator)"!(
     A:Rotator="0, 0, 0"
     B:Rotator=Noise by Sin!(amplitude:double=0.7, Offset:Vector, frequency Speed:Vector=@n3, Frequency control:Vector="0.1,1,1", Negative:bool) -> (ReturnValue:Rotator=~) {is=title K2Node_MacroInstance}
     Alpha:float=@n4
     bShortestPath:bool=false
    ) -> (ReturnValue:Rotator)) -> (Output_Get:Rotator)
   Set NoisRot2(
    execute=@n2.then_1
    NoisRot2:Rotator="Lerp (Rotator)"(
     A="0, 0, 0"
     B=Noise by Sin(amplitude=2, Offset="2,2,2", frequency Speed=@n3, Frequency control="0.1,1,1") -> (ReturnValue=~)
     Alpha=@n4
     bShortestPath=false
    )
   ) -> (Output_Get:Rotator)
   Set NoisRot3(
    execute=@n2.then_2
    NoisRot3:Rotator="Lerp (Rotator)"(
     A="0, 0, 0"
     B=Noise by Sin(amplitude=3, Offset="4.5,4.5,4.5", frequency Speed=@n3, Frequency control="0.1,1,1") -> (ReturnValue=~)
     Alpha=@n4
     bShortestPath=false
    )
   ) -> (Output_Get:Rotator)
   Set NoisRot4(
    execute=@n2.then_3
    NoisRot4:Rotator="Lerp (Rotator)"(
     A="0, 0, 0"
     B=Noise by Sin(amplitude=4, Offset="5,5,5", frequency Speed=@n3, Frequency control="0.1,1,1") -> (ReturnValue=~)
     Alpha=@n4
     bShortestPath=false
    )
   ) -> (Output_Get:Rotator)
   Event OnCharacterCustomizationLoaded()
   > Set IsWearingBackbling(IsWearingBackbling:bool=CheckForBackbling()) -> (Output_Get:bool)
   MakeVector(X=30, Y=40, Z=30)
   Get Gravity Override Params()
  ]}
 NoiseBySin {nodes [
   n1 = Inputs!() -> (amplitude:double=15, Offset:Vector=~, frequency Speed:Vector="8,8,8", Frequency control:Vector="0.1,1,1", Negative:bool=~)
   n2 = BreakVector(InVec=@n1.Offset)
   n3 = BreakVector(InVec=@n1.frequency Speed)
   n4 = GetRealTimeSeconds()
   n5 = float * float(A=@n4, B=@n3.Z) {is=title K2Node_PromotableOperator}
   n6 = BreakVector(InVec=@n1.Frequency control)
   n7 = float * float(A="Sin (Radians)"(A=float + float(A=@n5, B=@n2.Z)), B=@n6.Z)
   n8 = float * float(A=@n4, B=@n3.Y)
   n9 = float * float(A="Sin (Radians)"(A=float + float(A=@n8, B=@n2.Y)), B=@n6.Y)
   n10 = float * float(A=@n4, B=@n3)
   n11 = float * float(A="Sin (Radians)"(A=float + float(A=@n10, B=@n2)), B=@n6)
   n12 = Make Literal Float(Value=@n1)
   n13 = SelectFloat(A=-1, B=1, bPickA=@n1.Negative)
   Outputs!(ReturnValue:Rotator="Scale Rotator (Float)"(A:Rotator=MakeRotator(Roll=@n11, Pitch=@n9, Yaw=@n7), B:float=float * float(A=@n12, B=@n13)) -> (ReturnValue:Rotator=~) {is=title K2Node_PromotableOperator})
  ]}
]
variables [
 Gravity Override {type=Vector}
 Gravity Override Params {type=GravityOverrideParamsStruct}
 IsWearingBackbling {type=boolean}
 MyLandingJump {type=boolean}
 NoisRot1 {type=Rotator}
 NoisRot2 {type=Rotator}
 NoisRot3 {type=Rotator}
 NoisRot4 {type=Rotator}
]
